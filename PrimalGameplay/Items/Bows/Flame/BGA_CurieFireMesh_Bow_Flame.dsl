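; Blueprint actor for the Primal flame bow's fire mesh, which carries Curie fire fuel settings.
; On BeginPlay it calls the parent, then reads scalable-float rows for fire magnitude, propagation fuel and ignition radius.
; It writes these to CurieFuel, truncating fuel to an integer and using the single radius for both landscape ignition and igniting nearby materials.
; The construction script only defers to the parent.
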(event EventBeginPlay
  (bind _row_ignitionradius (Variables|Default|GetRow_IgnitionRadius))
  (bind _returnvalue (ScalableFloat|GetValueatLevel _row_ignitionradius))
  (|CallParentFunction)
  (Variables|CurieFuel|SetSelfFireMagnitude (ScalableFloat|GetValueatLevel (Variables|Default|GetRow_FireMagnitude)))
  (Variables|CurieFuel|SetSelfPropagationFuel (Math|Float|Truncate (ScalableFloat|GetValueatLevel (Variables|Default|GetRow_PropagationFuel))))
  (Variables|CurieFuel|SetSelfCurieLandscapeIgnitionRadius _returnvalue)
  (Variables|CurieFuel|SetIgniteNearbyMaterialsLandscapeIgnitionRadius _returnvalue))

(fn ConstructionScript
  (|CallParentFunction))
