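; Flame bow arrow projectile for Athena that adds fire and quest tracking to fully charged impacts.
; On hitting a building, it waits a configurable delay, then applies Curie fire with magnitude, propagation fuel and ignition radius from scalable float rows.
; Non-building hits go through player and wildlife casts and a gameplay tag check, which gates a fully-charged-hit stat event to the instigator's quest manager.
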(event EventBeginPlay
  (|CallParentFunction) ; This node is disabled and will not be called.\nDrag off pins to build functionality.
)

(event Collision|EventActorBeginOverlap
  (params (OtherActor :type Actor%20Object%20Reference))
  (|CallParentFunction OtherActor) ; This node is disabled and will not be called.\nDrag off pins to build functionality.
)

(event EventTick
  (params (DeltaSeconds :type Float%20%28single-precision%29))
  (|CallParentFunction DeltaSeconds) ; This node is disabled and will not be called.\nDrag off pins to build functionality.
)

(event EventOnFullyChargedImpact
  (params (Hit%20Result :type Hit%20Result%20Structure))
  (bind (_bblockinghit _binitialoverlap _time _distance _location _impactpoint _normal _impactnormal _physmat _hitactor) (Collision|BreakHitResult Hit%20Result))
  (|CallParentFunction Hit%20Result)
  ;[1 1 1] Apply fire directly if hitting a building
  (bind _asbuilding_static_mesh (Utilities|Casting|CastToBuildingSMActor _hitactor)
    (:then
      (Utilities|FlowControl|Delay (Variables|Default|GetDirectFireApplicationDelay))
      (Curie|ExecuteFireApplication :Instigator (Game|GetInstigator) :EffectCauser (Game|GetInstigator) :Target _asbuilding_static_mesh :Magnitude (ScalableFloat|GetValueatLevel (Variables|Default|GetRow_FireMagnitude)) :PropagationFuel (Math|Float|Truncate (ScalableFloat|GetValueatLevel (Variables|Default|GetRow_PropagationFuel))) :LandscapeIgnitionRadius (ScalableFloat|GetValueatLevel (Variables|Default|GetRow_IgnitionRadius))))
    (:CastFailed
      ;[1 1 1] Firing ComplexCustom Event for when a player is hit by a fully charged flame bow arrow
      (Utilities|Casting|CastToFortPlayerPawnAthena _hitactor
        (:then
          _)
        (:CastFailed
          (Utilities|Casting|CastToNPC_Pawn_Wildlife_Parent _hitactor)))
      (bind _returnvalue (GameplayTags|HasAnyMatchingGameplayTags(Message) _hitactor (GameplayTags|MakeGameplayTagContainerfromArray (Variables|Default|GetGameplay%20Tags))))
      (if _returnvalue
        (Quest|SendComplexCustomStatEvent (Quests|GetQuestManager (Fortnite|GetFortPlayerControllerfromActor (Variables|Actor|GetInstigator)) "Invalid") 0 "(GameplayTags=((TagName=\"Item.Weapon.Ranged.Bow.Flame.FullyChargedHit\")))" "(GameplayTags=)" 1)))))

(fn ConstructionScript
  (|CallParentFunction))
