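# Niagara idle effect for the flame bow, showing a small blue flame glow with sparking embers.
# Flame_Light bursts two mesh particles of differing size and opacity, with an offset blue light that flickers on a sine of emitter age.
# Embers spawns short-lived stretched sprites probabilistically in a tiny sphere, launched in a narrow cone with drag and fading from bright cyan-blue to deep blue.
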NiagaraSystem /PrimalGameplay/Items/Bows/Flame/FX/NS_FlameBow_Idle
dependencies {usedDataInterfaces [&NiagaraDataInterfaceColorCurve], usedDynamicInputs [], usedModules [&SystemState]}
emitters {
 Embers {
  emitterUpdateScript [
   &EmitterState(Inactive Response=::NewEnumerator1, Life Cycle Mode=::NewEnumerator1, Loop Duration=1, UseLoopDelay=false) {module=EmitterState001}
   &SpawnBurst_Instantaneous(Spawn Count=3, Spawn Probability=0.5, Use Spawn Probability=true)
   &SpawnRate(Spawn Probability=0.6, SpawnRate=60, Use Spawn Probability=true)
  ]
  particleSpawnScript [
   /Niagara/Modules/Spawn/Initialization/InitializeParticle(
    Color=<rgba 0.5 0.75 2 1>
    Lifetime=&UniformRangedFloat(Minimum=0.06, Maximum=0.1)
    Module.Write Color=true
    Module.Write Lifetime=true
    Module.Write Mass=true
    Module.Write Material Random=true
    Module.Write Position=true
    Module.Write Scale=false
    Module.Write SpriteRotation=false
    Module.Write SpriteSize=true
    Sprite Size=&Multiply_Vector2(
     A=<1 4>
     B=&Vector2DFromFloat(Value=&RandomRangeFloat(Minimum=0.1, Maximum=0.2))
    )
   )
   &SphereLocation(
    MaskSpawnGroup=false
    Override Seed=false
    Sphere Radius=0.5
    Surface Only=false
    UseNonUniformScale=false
    UseOffset=false
    UseSphereOrientation=false
   )
   &AddVelocityInCone(
    Cone Angle=12
    Cone Axis=&RotateVectorByQuaternion(Vector To Rotate=<1 0 0>, Quaternion=$Output.SkeletalMeshLocation.SampledRotation)
    Cone Axis Coordinate Space=::Simulation
    Module.Use Velocity Falloff On Cone Axis=false
    Velocity Strength=50
   )
  ]
  particleUpdateScript [
   &ParticleState()
   &Drag(Drag=3, Use Linear Drag=true, Use Rotational Drag=false)
   &SolveForcesAndVelocity(Clamp Velocity=false, Limit Acceleration=false)
   &ScaleColor(
    Scale RGB=&MakeVectorFromLinearColorRGB(Color=&ColorFromCurve(ColorCurve=DataInterface() {propertyValues {
        AlphaCurve {keys [<at 0 1 RCIM_Cubic>]}
        BlueCurve {keys [<at 0.3752 10>,<at 0.7114 1 RCIM_Cubic>]}
        GreenCurve {keys [<at 0.3752 1.2406>,<at 0.7114 0 RCIM_Cubic>]}
        RedCurve {keys [<at 0.3752 0>,<at 0.7114 0.0099 RCIM_Cubic>]}
       }}))
    ScaleA=true
    ScaleRGB=true
   )
  ]
  rendererClasses [&NiagaraSpriteRendererProperties]
  renderers [{rendererClass=&NiagaraSpriteRendererProperties}]
 }
 Flame_Light {
  emitterUpdateScript [
   &EmitterState(Inactive Response=::NewEnumerator1, Life Cycle Mode=::NewEnumerator1, Loop Duration=1, UseLoopDelay=false)
   &SpawnBurst_Instantaneous(Spawn Count=2, Use Spawn Probability=false)
  ]
  particleSpawnScript [
   &FN_SetSpawnGroup(Number of Spawn Groups=2)
   /SetVariables(Particles.SpawnGroupBool=HlslExpression() {hlslExpression="Particles.SpawnGroup == 0"}) {isSetParameters=true}
   &InitializeParticle(
    Color=<rgba 1 1 1 0.5>
    Mesh Scale=&Multiply_Vector(
     A=<0.2 0.02 0.02>
     B=&VectorFromFloat(Value=&MakeCustomFloatFromBool(Bool=$Particles.SpawnGroupBool, True Float=0.4, False Float=0.2))
    )
    Module.Write Color=true
    Module.Write Lifetime=true
    Module.Write Mass=true
    Module.Write Material Random=true
    Module.Write Position=true
    Module.Write Scale=true
    Module.Write SpriteRotation=false
    Module.Write SpriteSize=false
    Position=&SimulationPosition()
    Sprite Size=<25 25>
   )
   &Color(Color=&MakeCustomLinearColorFromBool(Bool=$Particles.SpawnGroupBool, True Color=<rgba 1 1 1 0.1>, False Color=<rgba 1 1 1 0.5>))
   /SetVariables(Particles.LightRadius=25) {isSetParameters=true}
   /SetVariables(Particles.LightPosition:Vector3f=&Add_Vector(A=Linked() {linkedVariable {name=Particles.Position, type=NiagaraPosition}}, B=<2 0 0>)) {isSetParameters=true}
  ]
  particleUpdateScript [
   &ParticleState()
   /SetVariables(Particles.LightColor=&MultiplyLinearColor(
     Linear Color=<rgba 0.125 0.6 1.5 1>
     Scale Factor=&MakeVector4_FromFloat(Float=&Lerp_Float(A=0.9, B=1, Alpha=&Sine(Normalized Angle=$Emitter.Age, Period=0.2, Scale=1, Bias=0)))
    )) {isSetParameters=true}
  ]
  rendererClasses [&NiagaraMeshRendererProperties, &NiagaraLightRendererProperties]
  renderers [
   {rendererClass=&NiagaraMeshRendererProperties, rendererIndex=0}
   {rendererClass=&NiagaraLightRendererProperties, rendererIndex=1}
  ]
 }
}
summary {
 systemName=NS_FlameBow_Idle
 emitters [
  {emitterName=Flame_Light, rendererClasses [&NiagaraMeshRendererProperties, &NiagaraLightRendererProperties]}
  {emitterName=Embers, rendererClasses [&NiagaraSpriteRendererProperties]}
 ]
 userVariables [
  User.Offset {type=Vector3f, description="", defaultValue {struct=Vector3f, value=<8 0 0>}}
  User.SkeletalMesh {
   type=NiagaraDataInterfaceSkeletalMesh
   defaultValue {
    struct=NiagaraExt_VariableValue_DataInterface
    value {dataInterface=&self:NiagaraDataInterfaceSkeletalMesh_0}
   }
  }
 ]
}
